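# Character-cosmetic material for the Unlucky Roll backpack that adds an animated, tinted emissive effect between the shared 2023 character PreFX and PostFX functions.
# A panning noise texture distorts the UVs of a scrolling storm-cloud texture, which is scaled, biased and tinted.
# The result multiplies an emissive texture.
# A static switch either blends it over the base emissive through a channel-selectable mask or replaces that emissive outright.
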Material /BRCosmetics/Accessories/FORT_Backpacks/Backpack_M_MED_UnluckyRoll/Materials/M_UnluckyRoll_Backpack
graph {
 properties {
  bUseMaterialAttributes=true
  bUsedWithMeshDeformer=true
  bUsedWithMorphTargets=true
  bUsedWithSkeletalMesh=true
  pixelDepthOffsetMode=PDOM_Legacy
  shadingModel=MSM_FromMaterialExpression
  usedShadingModels=Default Lit | Subsurface | Clear Coat | Subsurface Profile | Cloth
 }
 nodes [
  Constant() {r=-1}
  n1 = MaterialFunctionCall() {materialFunction=&MF_CharacterParent_2023_PreFX}
  n2 = GetMaterialAttributes(None=@n1.Result) {attributeGetTypes [#1]}
  n3 = TextureSampleParameter2D() {
   group=UnluckyRoll
   parameterName=EmissiveTex
   samplerType=SAMPLERTYPE_Masks
   texture=/BRCosmetics/Accessories/FORT_Backpacks/Backpack_M_MED_UnluckyRoll/Textures/T_UnluckyRoll_Backpack_E
  }
  n4 = Multiply(A=Multiply(A=@n3.RGB, B=NamedRerouteUsage()), B=ScalarParameter() {defaultValue=1, group=UnluckyRoll, parameterName=EmissiveIntensity})
  n5 = TextureSampleParameter2D() {
   group=UnluckyRoll
   parameterName=EmissiveMask
   texture=/BRCosmetics/Accessories/FORT_Backpacks/Backpack_M_MED_UnluckyRoll/Textures/T_UnluckyRoll_Backpack_FX
  }
  n6 = LinearInterpolate(A=@n2.Emissive Color, B=@n4, Alpha=ChannelMaskParameter(None=@n5.RGB) {group=UnluckyRoll, parameterName=EmissiveMaskChannel}.R)
  MaterialFunctionCall(
   Material=StaticSwitchParameter(
    True=SetMaterialAttributes(MaterialAttributes=@n2.MaterialAttributes, Emissive Color=@n6) {attributeSetTypes [#1]}
    False=SetMaterialAttributes(MaterialAttributes=@n2.MaterialAttributes, Emissive Color=@n4) {attributeSetTypes [#1]}
   ) {group=UnluckyRoll, parameterName=UseEmissiveMask}
   M=@n1.M
   Pre Skinned Local Position=@n1.M
   Distance Blend=@n1.M
   Fresnel=@n1.M
  ) {materialFunction=&MF_CharacterParent_2023_PostFX}
  n7 = TextureCoordinate()
  n8 = VectorParameter() {group=UnluckyRoll, parameterName=DistortionTiling, channelNames {a="", b="", g=Tile Y, r=Tile X}, defaultValue=<rgba 4 3 2 1>}
  n9 = ComponentMask(None=@n8) {g=true, r=true}
  n10 = VectorParameter() {group=UnluckyRoll, parameterName=DistortionSpeed, channelNames {a="", b="", g=Speed Y, r=Speed X}, defaultValue=<rgba -0.1 0.1 1 1>}
  n11 = TextureSampleParameter2D(UVs=Panner(Coordinate=Multiply(A=@n7, B=@n9), Speed=ComponentMask(None=@n10) {g=true, r=true})) {
   group=UnluckyRoll
   parameterName=DistortionTexture
   samplerType=SAMPLERTYPE_LinearColor
   texture=&T_NoisePack_LinearColor
  }
  NamedRerouteDeclaration(None=ChannelMaskParameter(None=@n11.RGB) {group=UnluckyRoll, parameterName=DistortionMaskChannel}.R) {name=DistortionMask, nodeColor=<rgba 0 1 0.0859 1>}
  NamedRerouteDeclaration(None=ComponentMask(None=ConstantBiasScale(None=@n11.RGB) {bias=-0.5, scale=2}) {g=true, r=true}) {name=UVDistortion, nodeColor=<rgba 1 0 0.75 1>}
  n12 = VectorParameter() {group=UnluckyRoll, parameterName=NoiseTiling, channelNames {a="", b="", g=Tile Y, r=Tile X}, defaultValue=<rgba 4 4 4 1>}
  n13 = Multiply(A=TextureCoordinate(), B=ComponentMask(None=@n12) {g=true, r=true})
  n14 = Multiply(A=ScalarParameter() {defaultValue=0.1, group=UnluckyRoll, parameterName=DistortionStrength}, B=NamedRerouteUsage())
  n15 = Time()
  n16 = VectorParameter() {group=UnluckyRoll, parameterName=NoiseSpeed, channelNames {a="", b="", g=Speed Y, r=Speed X}, defaultValue=<rgba 0.025 -0.05 1 1>}
  n17 = ComponentMask(None=@n16) {g=true, r=true}
  n18 = TextureSampleParameter2D(UVs=Add(A=Add(A=@n13, B=@n14), B=Multiply(A=@n15, B=@n17))) {
   group=UnluckyRoll
   parameterName=NoiseTexture
   samplerType=SAMPLERTYPE_LinearGrayscale
   texture=/Game/Characters/Player/Male/Medium/Bodies/M_MED_Hurricane/Textures/T_MED_Hurricane_StormCloud
  }
  n19 = Multiply(A=@n18.R, B=ScalarParameter() {defaultValue=1, group=UnluckyRoll, parameterName=NoiseRangeScale})
  n20 = ScalarParameter() {group=UnluckyRoll, parameterName=NoiseRangeBias}
  NamedRerouteDeclaration(None=Multiply(A=Add(A=@n19, B=@n20), B=VectorParameter() {group=UnluckyRoll, parameterName=EmissiveTint, defaultValue=<rgba 0 0 0 1>}.RGB)) {name=NoiseSample, nodeColor=<rgba 0 0.3438 1 1>}
 ]
}
