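# Physics asset for the UnluckyRoll backpack, named as a physics test, covering two hanging banners and a simplified character body.
# Each banner is an eight-capsule chain tied through a base body to the pack, with narrow swing limits that widen down the chain and zero twist.
# A separate body skeleton of pelvis, spine, arm and leg capsules uses free constraints, with oversized limb shapes that suggest collision proxies for the banners.
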PhysicsAsset /BRCosmetics/Accessories/FORT_Backpacks/Backpack_M_MED_UnluckyRoll/Meshes/M_MED_UnluckyRoll_Pack_PhysicsTest
bodies {
 base_banner_00_l [{shapeType=Capsule, length=0.505, radius=0.505, shapeName=base_banner_00_l_capsule, rotation=<rot 0 -90 0>}]
 base_banner_00_r [{shapeType=Capsule, length=0.505, radius=0.505, shapeName=base_banner_00_r_capsule, rotation=<rot 0 -90 0>}]
 body_calf_l [{shapeType=Capsule, length=10.5, radius=19.5, center=<-20 0 0>, rotation=<rot 90 0 0>}]
 body_calf_r [{shapeType=Capsule, length=10.5, radius=19.5, center=<20 0 0>, rotation=<rot 90 0 0>}]
 body_clavicle_l [{shapeType=Capsule, length=4, radius=4, rotation=<rot 90 0 0>}]
 body_clavicle_r [{shapeType=Capsule, length=4, radius=4, rotation=<rot 90 0 0>}]
 body_pelvis [{shapeType=Capsule, length=20, radius=15}]
 body_root [{shapeType=Capsule, length=4, radius=4, rotation=<rot 90 0 0>}]
 body_spine_01 [{shapeType=Capsule, length=4, radius=4, rotation=<rot 90 0 0>}]
 body_spine_02 [{shapeType=Capsule, length=4, radius=4, rotation=<rot 90 0 0>}]
 body_spine_03 [{shapeType=Capsule, length=4, radius=4, rotation=<rot 90 0 0>}]
 body_spine_04 [{shapeType=Capsule, length=4, radius=4, rotation=<rot 90 0 0>}]
 body_spine_05 [{shapeType=Capsule, length=4, radius=4, rotation=<rot 90 0 0>}]
 body_thigh_l [{shapeType=Capsule, length=10.5, radius=19.5, center=<-20 0 0>, rotation=<rot 90 0 0>}]
 body_thigh_r [{shapeType=Capsule, length=10.5, radius=19.5, center=<20 0 0>, rotation=<rot 90 0 0>}]
 body_upperarm_l [{shapeType=Capsule, length=25, radius=13.5, center=<20 12 0>, rotation=<rot 90 0 0>}]
 body_upperarm_r [{shapeType=Capsule, length=25, radius=13.5, center=<-20 -12 0>, rotation=<rot 90 0 0>}]
 dyn_banner_00_l [{shapeType=Capsule, length=7.5, radius=7, shapeName=dyn_banner_00_l_capsule, center=<8 0 0>, rotation=<rot -90 0 0>}]
 dyn_banner_00_r [{shapeType=Capsule, length=8, radius=7, shapeName=dyn_banner_00_r_capsule, center=<8 0 0>, rotation=<rot -90 0 0>}]
 dyn_banner_01_l [{shapeType=Capsule, length=9.5, radius=6.5, shapeName=dyn_banner_01_l_capsule, center=<8 0 0>, rotation=<rot -90 0 0>}]
 dyn_banner_01_r [{shapeType=Capsule, length=9, radius=6, shapeName=dyn_banner_01_r_capsule, center=<8 0 0>, rotation=<rot -90 0 0>}]
 dyn_banner_02_l [{shapeType=Capsule, length=9.5, radius=5.5, shapeName=dyn_banner_02_l_capsule, center=<8 0 0>, rotation=<rot -90 0 0>}]
 dyn_banner_02_r [{shapeType=Capsule, length=9.5, radius=5.5, shapeName=dyn_banner_02_r_capsule, center=<8 0 0>, rotation=<rot -90 0 0>}]
 dyn_banner_03_l [{shapeType=Capsule, length=6, radius=4, shapeName=dyn_banner_03_l_capsule, center=<8 0 0>, rotation=<rot -90 0 0>}]
 dyn_banner_03_r [{shapeType=Capsule, length=6, radius=4, shapeName=dyn_banner_03_r_capsule, center=<8 0 0>, rotation=<rot -90 0 0>}]
 dyn_banner_04_l [{shapeType=Capsule, length=4, radius=4, shapeName=dyn_banner_04_l_capsule, center=<8 0 0>, rotation=<rot -90 0 0>}]
 dyn_banner_04_r [{shapeType=Capsule, length=3, radius=3.5, shapeName=dyn_banner_04_r_capsule, center=<8 0 0>, rotation=<rot -90 0 0>}]
 dyn_banner_05_l [{shapeType=Capsule, length=4, radius=3.5, shapeName=dyn_banner_05_l_capsule, center=<8 0 0>, rotation=<rot -90 0 0>}]
 dyn_banner_05_r [{shapeType=Capsule, length=4, radius=3.5, shapeName=dyn_banner_05_r_capsule, center=<8 0 0>, rotation=<rot -90 0 0>}]
 dyn_banner_06_l [{shapeType=Capsule, length=4, radius=3.5, shapeName=dyn_banner_06_l_capsule, center=<8 0 0>, rotation=<rot -90 0 0>}]
 dyn_banner_06_r [{shapeType=Capsule, length=4, radius=2.5, shapeName=dyn_banner_06_r_capsule, center=<8 0 0>, rotation=<rot -90 0 0>}]
 dyn_banner_07_l [{shapeType=Capsule, length=4, radius=3, shapeName=dyn_banner_07_l_capsule, center=<8 0 0>, rotation=<rot -90 0 0>}]
 dyn_banner_07_r [{shapeType=Capsule, length=4, radius=2.5, shapeName=dyn_banner_07_r_capsule, center=<8 0 0>, rotation=<rot -90 0 0>}]
 pack [{shapeType=Capsule, length=0.505, radius=0.505, shapeName=pack_capsule, rotation=<rot 0 -90 0>}]
}
constraints [
 {
  bone1Name=base_banner_00_l
  bone2Name=pack
  swing1LimitDegrees=45
  swing1Motion=Free
  swing2LimitDegrees=45
  swing2Motion=Free
  twistLimitDegrees=45
  twistMotion=Free
 }
 {bone1Name=dyn_banner_00_l, bone2Name=base_banner_00_l, swing1LimitDegrees=8, swing2LimitDegrees=8, twistLimitDegrees=0}
 {bone1Name=dyn_banner_01_l, bone2Name=dyn_banner_00_l, swing1LimitDegrees=8, swing2LimitDegrees=8, twistLimitDegrees=0}
 {bone1Name=dyn_banner_02_l, bone2Name=dyn_banner_01_l, swing1LimitDegrees=12, swing2LimitDegrees=12, twistLimitDegrees=0}
 {bone1Name=dyn_banner_03_l, bone2Name=dyn_banner_02_l, swing1LimitDegrees=15, swing2LimitDegrees=15, twistLimitDegrees=0}
 {bone1Name=dyn_banner_04_l, bone2Name=dyn_banner_03_l, swing1LimitDegrees=15, swing2LimitDegrees=15, twistLimitDegrees=0}
 {bone1Name=dyn_banner_05_l, bone2Name=dyn_banner_04_l, swing1LimitDegrees=15, swing2LimitDegrees=15, twistLimitDegrees=0}
 {bone1Name=dyn_banner_06_l, bone2Name=dyn_banner_05_l, swing1LimitDegrees=15, swing2LimitDegrees=15, twistLimitDegrees=0}
 {bone1Name=dyn_banner_07_l, bone2Name=dyn_banner_06_l, swing1LimitDegrees=15, swing2LimitDegrees=15, twistLimitDegrees=0}
 {
  bone1Name=base_banner_00_r
  bone2Name=pack
  swing1LimitDegrees=45
  swing1Motion=Free
  swing2LimitDegrees=45
  swing2Motion=Free
  twistLimitDegrees=45
  twistMotion=Free
 }
 {bone1Name=dyn_banner_00_r, bone2Name=base_banner_00_r, swing1LimitDegrees=8, swing2LimitDegrees=8, twistLimitDegrees=0}
 {bone1Name=dyn_banner_01_r, bone2Name=dyn_banner_00_r, swing1LimitDegrees=8, swing2LimitDegrees=8, twistLimitDegrees=0}
 {bone1Name=dyn_banner_02_r, bone2Name=dyn_banner_01_r, swing1LimitDegrees=12, swing2LimitDegrees=12, twistLimitDegrees=0}
 {bone1Name=dyn_banner_03_r, bone2Name=dyn_banner_02_r, swing1LimitDegrees=15, swing2LimitDegrees=15, twistLimitDegrees=0}
 {bone1Name=dyn_banner_04_r, bone2Name=dyn_banner_03_r, swing1LimitDegrees=15, swing2LimitDegrees=15, twistLimitDegrees=0}
 {bone1Name=dyn_banner_05_r, bone2Name=dyn_banner_04_r, swing1LimitDegrees=15, swing2LimitDegrees=15, twistLimitDegrees=0}
 {bone1Name=dyn_banner_06_r, bone2Name=dyn_banner_05_r, swing1LimitDegrees=15, swing2LimitDegrees=15, twistLimitDegrees=0}
 {bone1Name=dyn_banner_07_r, bone2Name=dyn_banner_06_r, swing1LimitDegrees=15, swing2LimitDegrees=15, twistLimitDegrees=0}
 {
  bone1Name=body_pelvis
  bone2Name=body_root
  swing1LimitDegrees=45
  swing1Motion=Free
  swing2LimitDegrees=45
  swing2Motion=Free
  twistLimitDegrees=45
  twistMotion=Free
 }
 {
  bone1Name=body_spine_01
  bone2Name=body_pelvis
  swing1LimitDegrees=45
  swing1Motion=Free
  swing2LimitDegrees=45
  swing2Motion=Free
  twistLimitDegrees=45
  twistMotion=Free
 }
 {
  bone1Name=body_spine_02
  bone2Name=body_spine_01
  swing1LimitDegrees=45
  swing1Motion=Free
  swing2LimitDegrees=45
  swing2Motion=Free
  twistLimitDegrees=45
  twistMotion=Free
 }
 {
  bone1Name=body_spine_03
  bone2Name=body_spine_02
  swing1LimitDegrees=45
  swing1Motion=Free
  swing2LimitDegrees=45
  swing2Motion=Free
  twistLimitDegrees=45
  twistMotion=Free
 }
 {
  bone1Name=body_spine_04
  bone2Name=body_spine_03
  swing1LimitDegrees=45
  swing1Motion=Free
  swing2LimitDegrees=45
  swing2Motion=Free
  twistLimitDegrees=45
  twistMotion=Free
 }
 {
  bone1Name=body_spine_05
  bone2Name=body_spine_04
  swing1LimitDegrees=45
  swing1Motion=Free
  swing2LimitDegrees=45
  swing2Motion=Free
  twistLimitDegrees=45
  twistMotion=Free
 }
 {
  bone1Name=body_clavicle_l
  bone2Name=body_spine_05
  swing1LimitDegrees=45
  swing1Motion=Free
  swing2LimitDegrees=45
  swing2Motion=Free
  twistLimitDegrees=45
  twistMotion=Free
 }
 {
  bone1Name=body_upperarm_l
  bone2Name=body_clavicle_l
  swing1LimitDegrees=45
  swing1Motion=Free
  swing2LimitDegrees=45
  swing2Motion=Free
  twistLimitDegrees=45
  twistMotion=Free
 }
 {
  bone1Name=body_clavicle_r
  bone2Name=body_spine_05
  swing1LimitDegrees=45
  swing1Motion=Free
  swing2LimitDegrees=45
  swing2Motion=Free
  twistLimitDegrees=45
  twistMotion=Free
 }
 {
  bone1Name=body_upperarm_r
  bone2Name=body_clavicle_r
  swing1LimitDegrees=45
  swing1Motion=Free
  swing2LimitDegrees=45
  swing2Motion=Free
  twistLimitDegrees=45
  twistMotion=Free
 }
 {
  bone1Name=body_thigh_l
  bone2Name=body_pelvis
  swing1LimitDegrees=45
  swing1Motion=Free
  swing2LimitDegrees=45
  swing2Motion=Free
  twistLimitDegrees=45
  twistMotion=Free
 }
 {
  bone1Name=body_calf_l
  bone2Name=body_thigh_l
  swing1LimitDegrees=45
  swing1Motion=Free
  swing2LimitDegrees=45
  swing2Motion=Free
  twistLimitDegrees=45
  twistMotion=Free
 }
 {
  bone1Name=body_thigh_r
  bone2Name=body_pelvis
  swing1LimitDegrees=45
  swing1Motion=Free
  swing2LimitDegrees=45
  swing2Motion=Free
  twistLimitDegrees=45
  twistMotion=Free
 }
 {
  bone1Name=body_calf_r
  bone2Name=body_thigh_r
  swing1LimitDegrees=45
  swing1Motion=Free
  swing2LimitDegrees=45
  swing2Motion=Free
  twistLimitDegrees=45
  twistMotion=Free
 }
]
