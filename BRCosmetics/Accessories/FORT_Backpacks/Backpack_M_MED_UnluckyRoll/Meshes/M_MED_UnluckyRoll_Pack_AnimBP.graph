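# Animation Blueprint for the Unlucky Roll backpack cosmetic, adding secondary motion to its pack and banners.
# Each update copies parent-mesh socket transforms onto mirrored body_* bones in world space, then caches that pose.
# A crouch/movement-weighted pack offset feeds RigidBody, then serial AnimDynamics banner nodes, with state-driven wind gravity on RigidBody and some banners.
# It also drives material WPO cloth simulation, raising wind and turbulence while skydiving.
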AnimBlueprint /BRCosmetics/Accessories/FORT_Backpacks/Backpack_M_MED_UnluckyRoll/Meshes/M_MED_UnluckyRoll_Pack_AnimBP
target_skeleton=&/M_MED_UnluckyRoll_Pack_Skeleton
graphs [
 AnimGraph {nodes [
   "Anim Dynamics - Chain: dyn_banner_03_extra_01_r -> dyn_banner_03_extra_03_r"!(ComponentPose, Alpha=0.6, GravityOverride:Vector="0,0,0", ExternalForce="0,0,0") -> (Pose) {
    angularDampingOverride=1
    angularSpringConstant=10000
    bAngularSpring=true
    bChain=true
    bOverrideAngularDamping=true
    bUseGravityOverride=true
    gravityScale=0.35
    simulationSpace=RootRelative
    alphaBoolBlend {blendOption=Linear}
    boundBone {boneName=dyn_banner_03_extra_01_r}
    chainEnd {boneName=dyn_banner_03_extra_03_r}
    componentAppliedLinearAccClamp=<3000 3000 3000>
    componentLinearAccScale=<0.3 0.3 0.3>
    componentLinearVelScale=<3 0.3 0.3>
    physicsBodyDefinitions [
     {
      boundBone {boneName=dyn_banner_03_extra_01_r}
      boxExtents=<20 20 20>
      constraintSetup {angularLimitsMax=<0 0 0>, angularLimitsMin=<0 0 0>, angularTarget=<0 0 0>}
      localJointOffset=<0 0 0>
     }
     {
      boundBone {boneName=dyn_banner_03_extra_02_r}
      boxExtents=<8 8 8>
      constraintSetup {angularLimitsMax=<0 0 0>, angularLimitsMin=<0 0 0>, angularTarget=<0 0 0>}
      localJointOffset=<0 0 0>
     }
     {
      boundBone {boneName=dyn_banner_03_extra_03_r}
      boxExtents=<5 5 5>
      constraintSetup {angularLimitsMax=<0 0 0>, angularLimitsMin=<0 0 0>, angularTarget=<0 0 0>}
      localJointOffset=<0 0 0>
     }
    ]
   }
   "Anim Dynamics - Chain: dyn_banner_03_extra_01_l -> dyn_banner_03_extra_03_l"!(ComponentPose, Alpha=0.6, GravityOverride:Vector="0,0,0", ExternalForce="0,0,0") -> (Pose) {
    angularDampingOverride=1
    angularSpringConstant=10000
    bAngularSpring=true
    bChain=true
    bOverrideAngularDamping=true
    bUseGravityOverride=true
    gravityScale=0.35
    simulationSpace=RootRelative
    alphaBoolBlend {blendOption=Linear}
    boundBone {boneName=dyn_banner_03_extra_01_l}
    chainEnd {boneName=dyn_banner_03_extra_03_l}
    componentAppliedLinearAccClamp=<3000 3000 3000>
    componentLinearAccScale=<0.3 0.3 0.3>
    componentLinearVelScale=<3 0.3 0.3>
    physicsBodyDefinitions [
     {
      boundBone {boneName=dyn_banner_03_extra_01_l}
      boxExtents=<20 20 20>
      constraintSetup {angularLimitsMax=<10 20 15>, angularLimitsMin=<-10 -20 15>, angularTarget=<1 0 0>}
      localJointOffset=<4 0 0>
     }
     {
      boundBone {boneName=dyn_banner_03_extra_02_l}
      boxExtents=<8 8 8>
      constraintSetup {angularLimitsMax=<10 20 15>, angularLimitsMin=<-10 -20 15>, angularTarget=<1 0 0>}
      localJointOffset=<4 0 0>
     }
     {
      boundBone {boneName=dyn_banner_03_extra_03_l}
      boxExtents=<5 5 5>
      constraintSetup {angularLimitsMax=<10 20 15>, angularLimitsMin=<-10 -20 15>, angularTarget=<1 0 0>}
      localJointOffset=<4 0 0>
     }
    ]
   }
   n1 = "Transform (Modify) Bone - Bone: pack"(
    ComponentPose=Local To Component(LocalPose=Use cached pose 'Initial_Cache'())
    Alpha=Get BackpackAlpha() -> (BackpackAlpha:double)
    Translation="0,-1.8,0"
    Rotation="0,0,0"
   ) {
    translationMode=BMM_Additive
    translationSpace=BCS_BoneSpace
    alphaBoolBlend {blendOption=Linear}
    alphaScaleBiasClamp {bInterpResult=true}
    boneToModify {boneName=pack}
    translation {y=-1.8}
   }
   n2 = RigidBody!(ComponentPose=@n1, Alpha, OverrideWorldGravity:Vector=Get GravityOverride(), ExternalForce) -> (Pose) {
    bClampLinearTranslationLimitToRefPose=true
    bOverrideWorldGravity=true
    overridePhysicsAsset=&/M_MED_UnluckyRoll_Pack_PhysicsTest
    simulationSpace=BaseBoneSpace
    alphaBoolBlend {blendOption=Linear}
    baseBoneRef {boneName=pack}
    componentAppliedLinearAccClamp=<8000 8000 8000>
    componentLinearAccScale=<0.4 0.4 0.4>
    componentLinearVelScale=<0.4 0.4 0.4>
    simSpaceSettings {
     maxAngularAcceleration=1000
     maxAngularVelocity=3
     maxLinearAcceleration=700
     maxLinearVelocity=500
     velocityScaleZ=0
     worldAlpha=0.25
     externalLinearDragV=<0.2 0.2>
    }
   }
   n3 = Anim Dynamics - Bone: dyn_banner_02_extra_l!(ComponentPose=@n2, Alpha=1, GravityOverride:Vector="0,0,0", ExternalForce="0,0,0") -> (Pose) {
    angularDampingOverride=1
    angularSpringConstant=10000
    bAngularSpring=true
    bOverrideAngularDamping=true
    bUseGravityOverride=true
    gravityScale=0.35
    simulationSpace=RootRelative
    alphaBoolBlend {blendOption=Linear}
    boundBone {boneName=dyn_banner_02_extra_l}
    componentAppliedLinearAccClamp=<3000 3000 3000>
    componentLinearAccScale=<0.3 0.3 0.3>
    componentLinearVelScale=<3 0.3 0.3>
    physicsBodyDefinitions [{
      boundBone {boneName=dyn_banner_02_extra_l}
      boxExtents=<10 10 10>
      constraintSetup {angularLimitsMax=<0 0 0>, angularLimitsMin=<0 0 0>, angularTarget=<0 0 0>}
      localJointOffset=<0 0 0>
     }]
   }
   n4 = Get GravityOverride()
   n5 = Anim Dynamics - Bone: dyn_banner_03_extra_01_l!(ComponentPose=@n3, Alpha=1, GravityOverride:Vector=@n4, ExternalForce="0,0,0") -> (Pose) {
    angularDampingOverride=1
    angularSpringConstant=10000
    bAngularSpring=true
    bOverrideAngularDamping=true
    bUseGravityOverride=true
    gravityScale=0.35
    simulationSpace=RootRelative
    alphaBoolBlend {blendOption=Linear}
    boundBone {boneName=dyn_banner_03_extra_01_l}
    componentLinearAccScale=<1 1 1>
    componentLinearVelScale=<1 1 1>
    gravityOverride {z=-980}
    physicsBodyDefinitions [{
      boundBone {boneName=dyn_banner_03_extra_01_l}
      boxExtents=<20 30 30>
      constraintSetup {angularLimitsMax=<15 5 15>, angularLimitsMin=<-15 -5 -15>, angularTarget=<1 0 0>}
      localJointOffset=<4 0 0>
     }]
   }
   n6 = Anim Dynamics - Bone: dyn_banner_03_extra_02_l!(ComponentPose=@n5, Alpha=1, GravityOverride:Vector=@n4, ExternalForce="0,0,0") -> (Pose) {
    angularDampingOverride=1
    angularSpringConstant=10000
    bAngularSpring=true
    bOverrideAngularDamping=true
    bUseGravityOverride=true
    gravityScale=0.35
    simulationSpace=RootRelative
    alphaBoolBlend {blendOption=Linear}
    boundBone {boneName=dyn_banner_03_extra_02_l}
    componentLinearAccScale=<1 1 1>
    componentLinearVelScale=<1 1 1>
    gravityOverride {z=-980}
    physicsBodyDefinitions [{
      boundBone {boneName=dyn_banner_03_extra_02_l}
      boxExtents=<20 30 30>
      constraintSetup {angularLimitsMax=<15 5 15>, angularLimitsMin=<-15 -5 -15>, angularTarget=<1 0 0>}
      localJointOffset=<4 0 0>
     }]
   }
   n7 = Anim Dynamics - Bone: dyn_banner_03_extra_03_l!(ComponentPose=@n6, Alpha=1, GravityOverride:Vector=@n4, ExternalForce="0,0,0") -> (Pose) {
    angularDampingOverride=1
    angularSpringConstant=10000
    bAngularSpring=true
    bOverrideAngularDamping=true
    bUseGravityOverride=true
    gravityScale=0.35
    simulationSpace=RootRelative
    alphaBoolBlend {blendOption=Linear}
    boundBone {boneName=dyn_banner_03_extra_03_l}
    componentLinearAccScale=<1 1 1>
    componentLinearVelScale=<1 1 1>
    gravityOverride {z=-980}
    physicsBodyDefinitions [{
      boundBone {boneName=dyn_banner_03_extra_03_l}
      boxExtents=<20 30 30>
      constraintSetup {angularLimitsMax=<15 5 15>, angularLimitsMin=<-15 -5 -15>, angularTarget=<1 0 0>}
      localJointOffset=<4 0 0>
     }]
   }
   n8 = "Anim Dynamics - Chain: dyn_banner_03_extra_04_l -> dyn_banner_03_extra_05_l"!(ComponentPose=@n7, Alpha=1, GravityOverride:Vector="0,0,0", ExternalForce="0,0,0") -> (Pose) {
    angularDampingOverride=1
    angularSpringConstant=10000
    bAngularSpring=true
    bChain=true
    bOverrideAngularDamping=true
    bUseGravityOverride=true
    gravityScale=0.35
    simulationSpace=RootRelative
    alphaBoolBlend {blendOption=Linear}
    boundBone {boneName=dyn_banner_03_extra_04_l}
    chainEnd {boneName=dyn_banner_03_extra_05_l}
    componentAppliedLinearAccClamp=<3000 3000 3000>
    componentLinearAccScale=<0.3 0.3 0.3>
    componentLinearVelScale=<3 0.3 0.3>
    physicsBodyDefinitions [
     {
      boundBone {boneName=dyn_banner_03_extra_04_l}
      boxExtents=<10 10 10>
      constraintSetup {angularLimitsMax=<0 0 0>, angularLimitsMin=<0 0 0>, angularTarget=<0 0 0>}
      localJointOffset=<0 0 0>
     }
     {
      boundBone {boneName=dyn_banner_03_extra_05_l}
      boxExtents=<10 10 10>
      constraintSetup {angularLimitsMax=<0 0 0>, angularLimitsMin=<0 0 0>, angularTarget=<0 0 0>}
      localJointOffset=<0 0 0>
     }
    ]
   }
   n9 = "Anim Dynamics - Chain: dyn_banner_03_extra_06_l -> dyn_banner_03_extra_07_l"!(ComponentPose=@n8, Alpha=1, GravityOverride:Vector="0,0,0", ExternalForce="0,0,0") -> (Pose) {
    angularDampingOverride=1
    angularSpringConstant=10000
    bAngularSpring=true
    bChain=true
    bOverrideAngularDamping=true
    bUseGravityOverride=true
    gravityScale=0.35
    simulationSpace=RootRelative
    alphaBoolBlend {blendOption=Linear}
    boundBone {boneName=dyn_banner_03_extra_06_l}
    chainEnd {boneName=dyn_banner_03_extra_07_l}
    componentAppliedLinearAccClamp=<3000 3000 3000>
    componentLinearAccScale=<0.3 0.3 0.3>
    componentLinearVelScale=<3 0.3 0.3>
    physicsBodyDefinitions [
     {
      boundBone {boneName=dyn_banner_03_extra_06_l}
      boxExtents=<10 10 10>
      constraintSetup {angularLimitsMax=<0 0 0>, angularLimitsMin=<0 0 0>, angularTarget=<0 0 0>}
      localJointOffset=<0 0 0>
     }
     {
      boundBone {boneName=dyn_banner_03_extra_07_l}
      boxExtents=<10 10 10>
      constraintSetup {angularLimitsMax=<0 0 0>, angularLimitsMin=<0 0 0>, angularTarget=<0 0 0>}
      localJointOffset=<0 0 0>
     }
    ]
   }
   n10 = "Anim Dynamics - Chain: dyn_banner_04_extra_01_l -> dyn_banner_04_extra_02_l"!(ComponentPose=@n9, Alpha=1, GravityOverride:Vector="0,0,0", ExternalForce="0,0,0") -> (Pose) {
    angularDampingOverride=1
    angularSpringConstant=10000
    bAngularSpring=true
    bChain=true
    bOverrideAngularDamping=true
    bUseGravityOverride=true
    gravityScale=0.35
    simulationSpace=RootRelative
    alphaBoolBlend {blendOption=Linear}
    boundBone {boneName=dyn_banner_04_extra_01_l}
    chainEnd {boneName=dyn_banner_04_extra_02_l}
    componentAppliedLinearAccClamp=<3000 3000 3000>
    componentLinearAccScale=<0.3 0.3 0.3>
    componentLinearVelScale=<3 0.3 0.3>
    physicsBodyDefinitions [
     {
      boundBone {boneName=dyn_banner_04_extra_01_l}
      boxExtents=<8 8 8>
      constraintSetup {angularLimitsMax=<0 0 0>, angularLimitsMin=<0 0 0>, angularTarget=<0 0 0>}
      localJointOffset=<0 0 0>
     }
     {
      boundBone {boneName=dyn_banner_04_extra_02_l}
      boxExtents=<8 8 8>
      constraintSetup {angularLimitsMax=<0 0 0>, angularLimitsMin=<0 0 0>, angularTarget=<0 0 0>}
      localJointOffset=<0 0 0>
     }
    ]
   }
   n11 = "Anim Dynamics - Chain: dyn_banner_05_extra_01_l -> dyn_banner_05_extra_02_l"!(ComponentPose=@n10, Alpha=1, GravityOverride:Vector="0,0,0", ExternalForce="0,0,0") -> (Pose) {
    angularDampingOverride=1
    angularSpringConstant=10000
    bAngularSpring=true
    bChain=true
    bOverrideAngularDamping=true
    bUseGravityOverride=true
    gravityScale=0.35
    simulationSpace=RootRelative
    alphaBoolBlend {blendOption=Linear}
    boundBone {boneName=dyn_banner_05_extra_01_l}
    chainEnd {boneName=dyn_banner_05_extra_02_l}
    componentAppliedLinearAccClamp=<3000 3000 3000>
    componentLinearAccScale=<0.3 0.3 0.3>
    componentLinearVelScale=<3 0.3 0.3>
    physicsBodyDefinitions [
     {
      boundBone {boneName=dyn_banner_05_extra_01_l}
      boxExtents=<10 10 10>
      constraintSetup {angularLimitsMax=<0 0 0>, angularLimitsMin=<0 0 0>, angularTarget=<0 0 0>}
      localJointOffset=<0 0 0>
     }
     {
      boundBone {boneName=dyn_banner_05_extra_02_l}
      boxExtents=<10 10 10>
      constraintSetup {angularLimitsMax=<0 0 0>, angularLimitsMin=<0 0 0>, angularTarget=<0 0 0>}
      localJointOffset=<0 0 0>
     }
    ]
   }
   n12 = Anim Dynamics - Bone: dyn_banner_02_extra_r!(ComponentPose=@n11, Alpha=1, GravityOverride:Vector="0,0,0", ExternalForce="0,0,0") -> (Pose) {
    angularDampingOverride=1
    angularSpringConstant=20000
    bAngularSpring=true
    bOverrideAngularDamping=true
    bUseGravityOverride=true
    gravityScale=0.35
    simulationSpace=RootRelative
    alphaBoolBlend {blendOption=Linear}
    boundBone {boneName=dyn_banner_02_extra_r}
    componentAppliedLinearAccClamp=<3000 3000 3000>
    componentLinearAccScale=<0.3 0.3 0.3>
    componentLinearVelScale=<3 0.3 0.3>
    physicsBodyDefinitions [{
      boundBone {boneName=dyn_banner_02_extra_r}
      boxExtents=<10 10 10>
      constraintSetup {angularLimitsMax=<0 0 0>, angularLimitsMin=<0 0 0>, angularTarget=<0 0 0>}
      localJointOffset=<0 0 0>
     }]
   }
   n13 = Get GravityOverride()
   n14 = Anim Dynamics - Bone: dyn_banner_03_extra_01_r!(ComponentPose=@n12, Alpha=1, GravityOverride:Vector=@n13, ExternalForce="0,0,0") -> (Pose) {
    angularDampingOverride=1
    angularSpringConstant=10000
    bAngularSpring=true
    bOverrideAngularDamping=true
    bUseGravityOverride=true
    gravityScale=0.35
    simulationSpace=RootRelative
    alphaBoolBlend {blendOption=Linear}
    boundBone {boneName=dyn_banner_03_extra_01_r}
    componentLinearAccScale=<1 1 1>
    componentLinearVelScale=<1 1 1>
    gravityOverride {z=-980}
    physicsBodyDefinitions [{
      boundBone {boneName=dyn_banner_03_extra_01_r}
      boxExtents=<20 30 30>
      constraintSetup {angularLimitsMax=<15 5 15>, angularLimitsMin=<-15 -5 -15>, angularTarget=<1 0 0>}
      localJointOffset=<4 0 0>
     }]
   }
   n15 = Anim Dynamics - Bone: dyn_banner_03_extra_02_r!(ComponentPose=@n14, Alpha=1, GravityOverride:Vector=@n13, ExternalForce="0,0,0") -> (Pose) {
    angularDampingOverride=1
    angularSpringConstant=10000
    bAngularSpring=true
    bOverrideAngularDamping=true
    bUseGravityOverride=true
    gravityScale=0.35
    simulationSpace=RootRelative
    alphaBoolBlend {blendOption=Linear}
    boundBone {boneName=dyn_banner_03_extra_02_r}
    componentLinearAccScale=<1 1 1>
    componentLinearVelScale=<1 1 1>
    gravityOverride {z=-980}
    physicsBodyDefinitions [{
      boundBone {boneName=dyn_banner_03_extra_02_r}
      boxExtents=<20 30 30>
      constraintSetup {angularLimitsMax=<15 5 15>, angularLimitsMin=<-15 -5 -15>, angularTarget=<1 0 0>}
      localJointOffset=<4 0 0>
     }]
   }
   n16 = Anim Dynamics - Bone: dyn_banner_03_extra_03_r!(ComponentPose=@n15, Alpha=1, GravityOverride:Vector=@n13, ExternalForce="0,0,0") -> (Pose) {
    angularDampingOverride=1
    angularSpringConstant=10000
    bAngularSpring=true
    bOverrideAngularDamping=true
    bUseGravityOverride=true
    gravityScale=0.35
    simulationSpace=RootRelative
    alphaBoolBlend {blendOption=Linear}
    boundBone {boneName=dyn_banner_03_extra_03_r}
    componentLinearAccScale=<1 1 1>
    componentLinearVelScale=<1 1 1>
    gravityOverride {z=-980}
    physicsBodyDefinitions [{
      boundBone {boneName=dyn_banner_03_extra_03_r}
      boxExtents=<20 30 30>
      constraintSetup {angularLimitsMax=<15 5 15>, angularLimitsMin=<-15 -5 -15>, angularTarget=<1 0 0>}
      localJointOffset=<4 0 0>
     }]
   }
   n17 = "Anim Dynamics - Chain: dyn_banner_03_extra_04_r -> dyn_banner_03_extra_05_r"!(ComponentPose=@n16, Alpha=1, GravityOverride:Vector="0,0,0", ExternalForce="0,0,0") -> (Pose) {
    angularDampingOverride=1
    angularSpringConstant=10000
    bAngularSpring=true
    bChain=true
    bOverrideAngularDamping=true
    bUseGravityOverride=true
    gravityScale=0.35
    simulationSpace=RootRelative
    alphaBoolBlend {blendOption=Linear}
    boundBone {boneName=dyn_banner_03_extra_04_r}
    chainEnd {boneName=dyn_banner_03_extra_05_r}
    componentAppliedLinearAccClamp=<3000 3000 3000>
    componentLinearAccScale=<0.3 0.3 0.3>
    componentLinearVelScale=<3 0.3 0.3>
    physicsBodyDefinitions [
     {
      boundBone {boneName=dyn_banner_03_extra_04_r}
      boxExtents=<10 10 10>
      constraintSetup {angularLimitsMax=<0 0 0>, angularLimitsMin=<0 0 0>, angularTarget=<0 0 0>}
      localJointOffset=<0 0 0>
     }
     {
      boundBone {boneName=dyn_banner_03_extra_05_r}
      boxExtents=<10 10 10>
      constraintSetup {angularLimitsMax=<0 0 0>, angularLimitsMin=<0 0 0>, angularTarget=<0 0 0>}
      localJointOffset=<0 0 0>
     }
    ]
   }
   n18 = "Anim Dynamics - Chain: dyn_banner_03_extra_06_r -> dyn_banner_03_extra_07_r"!(ComponentPose=@n17, Alpha=1, GravityOverride:Vector="0,0,0", ExternalForce="0,0,0") -> (Pose) {
    angularDampingOverride=1
    angularSpringConstant=10000
    bAngularSpring=true
    bChain=true
    bOverrideAngularDamping=true
    bUseGravityOverride=true
    gravityScale=0.35
    simulationSpace=RootRelative
    alphaBoolBlend {blendOption=Linear}
    boundBone {boneName=dyn_banner_03_extra_06_r}
    chainEnd {boneName=dyn_banner_03_extra_07_r}
    componentAppliedLinearAccClamp=<3000 3000 3000>
    componentLinearAccScale=<0.3 0.3 0.3>
    componentLinearVelScale=<3 0.3 0.3>
    physicsBodyDefinitions [
     {
      boundBone {boneName=dyn_banner_03_extra_06_r}
      boxExtents=<10 10 10>
      constraintSetup {angularLimitsMax=<0 0 0>, angularLimitsMin=<0 0 0>, angularTarget=<0 0 0>}
      localJointOffset=<0 0 0>
     }
     {
      boundBone {boneName=dyn_banner_03_extra_07_r}
      boxExtents=<10 10 10>
      constraintSetup {angularLimitsMax=<0 0 0>, angularLimitsMin=<0 0 0>, angularTarget=<0 0 0>}
      localJointOffset=<0 0 0>
     }
    ]
   }
   n19 = "Anim Dynamics - Chain: dyn_banner_04_extra_01_r -> dyn_banner_04_extra_02_r"!(ComponentPose=@n18, Alpha=1, GravityOverride:Vector="0,0,0", ExternalForce="0,0,0") -> (Pose) {
    angularDampingOverride=1
    angularSpringConstant=10000
    bAngularSpring=true
    bChain=true
    bOverrideAngularDamping=true
    bUseGravityOverride=true
    gravityScale=0.35
    simulationSpace=RootRelative
    alphaBoolBlend {blendOption=Linear}
    boundBone {boneName=dyn_banner_04_extra_01_r}
    chainEnd {boneName=dyn_banner_04_extra_02_r}
    componentAppliedLinearAccClamp=<3000 3000 3000>
    componentLinearAccScale=<0.3 0.3 0.3>
    componentLinearVelScale=<3 0.3 0.3>
    physicsBodyDefinitions [
     {
      boundBone {boneName=dyn_banner_04_extra_01_r}
      boxExtents=<10 10 10>
      constraintSetup {angularLimitsMax=<0 0 0>, angularLimitsMin=<0 0 0>, angularTarget=<0 0 0>}
      localJointOffset=<0 0 0>
     }
     {
      boundBone {boneName=dyn_banner_04_extra_02_r}
      boxExtents=<10 10 10>
      constraintSetup {angularLimitsMax=<0 0 0>, angularLimitsMin=<0 0 0>, angularTarget=<0 0 0>}
      localJointOffset=<0 0 0>
     }
    ]
   }
   n20 = "Anim Dynamics - Chain: dyn_banner_05_extra_01_r -> dyn_banner_05_extra_02_r"!(ComponentPose=@n19, Alpha=1, GravityOverride:Vector="0,0,0", ExternalForce="0,0,0") -> (Pose) {
    angularDampingOverride=1
    angularSpringConstant=10000
    bAngularSpring=true
    bChain=true
    bOverrideAngularDamping=true
    bUseGravityOverride=true
    gravityScale=0.35
    simulationSpace=RootRelative
    alphaBoolBlend {blendOption=Linear}
    boundBone {boneName=dyn_banner_05_extra_01_r}
    chainEnd {boneName=dyn_banner_05_extra_02_r}
    componentAppliedLinearAccClamp=<3000 3000 3000>
    componentLinearAccScale=<0.3 0.3 0.3>
    componentLinearVelScale=<3 0.3 0.3>
    physicsBodyDefinitions [
     {
      boundBone {boneName=dyn_banner_05_extra_01_r}
      boxExtents=<10 10 10>
      constraintSetup {angularLimitsMax=<0 0 0>, angularLimitsMin=<0 0 0>, angularTarget=<0 0 0>}
      localJointOffset=<0 0 0>
     }
     {
      boundBone {boneName=dyn_banner_05_extra_02_r}
      boxExtents=<10 10 10>
      constraintSetup {angularLimitsMax=<0 0 0>, angularLimitsMin=<0 0 0>, angularTarget=<0 0 0>}
      localJointOffset=<0 0 0>
     }
    ]
   }
   Output Pose(Result=Component To Local(ComponentPose=@n20)) {layerGroup=None}
   n21 = "Transform (Modify) Bone - Bone: body_root"(
    ComponentPose=Local To Component(LocalPose=Input Pose())
    Alpha=1
    Translation=Get body_root_location() -> (body_root_location:Vector)
    Rotation=Get body_root_rotation() -> (body_root_rotation:Rotator)
    Scale="1,1,1"
   ) {
    rotationMode=BMM_Replace
    rotationSpace=BCS_WorldSpace
    scaleSpace=BCS_BoneSpace
    translationMode=BMM_Replace
    translationSpace=BCS_WorldSpace
    alphaBoolBlend {blendOption=Linear}
    alphaScaleBiasClamp {bInterpResult=true, interpSpeedDecreasing=1, interpSpeedIncreasing=1, inRange {max=0, min=-40}, outRange {max=0, min=1}}
    boneToModify {boneName=body_root}
   }
   n22 = "Transform (Modify) Bone - Bone: body_pelvis"(
    ComponentPose=@n21
    Alpha=1
    Translation=Get pelvis_location() -> (pelvis_location:Vector)
    Rotation=Get pelvis_rotation() -> (pelvis_rotation:Rotator)
    Scale="1,1,1"
   ) {
    rotationMode=BMM_Replace
    rotationSpace=BCS_WorldSpace
    scaleSpace=BCS_BoneSpace
    translationMode=BMM_Replace
    translationSpace=BCS_WorldSpace
    alphaBoolBlend {blendOption=Linear}
    alphaScaleBiasClamp {inRange {max=0, min=-40}, outRange {max=0, min=1}}
    boneToModify {boneName=body_pelvis}
   }
   n23 = "Transform (Modify) Bone - Bone: body_thigh_l"(
    ComponentPose=@n22
    Alpha=1
    Translation=Get thigh_left_location() -> (thigh_left_location:Vector)
    Rotation=Get thigh_left_rotation() -> (thigh_left_rotation:Rotator)
    Scale="1,1,1"
   ) {
    rotationMode=BMM_Replace
    rotationSpace=BCS_WorldSpace
    scaleSpace=BCS_BoneSpace
    translationMode=BMM_Replace
    translationSpace=BCS_WorldSpace
    alphaBoolBlend {blendOption=Linear}
    alphaScaleBiasClamp {inRange {max=0, min=-40}, outRange {max=0, min=1}}
    boneToModify {boneName=body_thigh_l}
   }
   n24 = "Transform (Modify) Bone - Bone: body_thigh_r"(
    ComponentPose=@n23
    Alpha=1
    Translation=Get thigh_right_location() -> (thigh_right_location:Vector)
    Rotation=Get thigh_right_rotation() -> (thigh_right_rotation:Rotator)
    Scale="1,1,1"
   ) {
    rotationMode=BMM_Replace
    rotationSpace=BCS_WorldSpace
    scaleSpace=BCS_BoneSpace
    translationMode=BMM_Replace
    translationSpace=BCS_WorldSpace
    alphaBoolBlend {blendOption=Linear}
    alphaScaleBiasClamp {inRange {max=0, min=-40}, outRange {max=0, min=1}}
    boneToModify {boneName=body_thigh_r}
   }
   n25 = "Transform (Modify) Bone - Bone: body_calf_l"(
    ComponentPose=@n24
    Alpha=1
    Translation=Get calf_left_location() -> (calf_left_location:Vector)
    Rotation=Get calf_left_rotation() -> (calf_left_rotation:Rotator)
    Scale="1,1,1"
   ) {
    rotationMode=BMM_Replace
    rotationSpace=BCS_WorldSpace
    scaleSpace=BCS_BoneSpace
    translationMode=BMM_Replace
    translationSpace=BCS_WorldSpace
    alphaBoolBlend {blendOption=Linear}
    alphaScaleBiasClamp {inRange {max=0, min=-40}, outRange {max=0, min=1}}
    boneToModify {boneName=body_calf_l}
   }
   n26 = "Transform (Modify) Bone - Bone: body_calf_r"(
    ComponentPose=@n25
    Alpha=1
    Translation=Get calf_right_location() -> (calf_right_location:Vector)
    Rotation=Get calf_right_rotation() -> (calf_right_rotation:Rotator)
    Scale="1,1,1"
   ) {
    rotationMode=BMM_Replace
    rotationSpace=BCS_WorldSpace
    scaleSpace=BCS_BoneSpace
    translationMode=BMM_Replace
    translationSpace=BCS_WorldSpace
    alphaBoolBlend {blendOption=Linear}
    alphaScaleBiasClamp {inRange {max=0, min=-40}, outRange {max=0, min=1}}
    boneToModify {boneName=body_calf_r}
   }
   n27 = "Transform (Modify) Bone - Bone: body_spine_01"(
    ComponentPose=@n26
    Alpha=1
    Translation=Get spine_01_location() -> (spine_01_location:Vector)
    Rotation=Get spine_01_rotation() -> (spine_01_rotation:Rotator)
    Scale="1,1,1"
   ) {
    rotationMode=BMM_Replace
    rotationSpace=BCS_WorldSpace
    scaleSpace=BCS_BoneSpace
    translationMode=BMM_Replace
    translationSpace=BCS_WorldSpace
    alphaBoolBlend {blendOption=Linear}
    alphaScaleBiasClamp {bInterpResult=true, interpSpeedDecreasing=1, interpSpeedIncreasing=1, inRange {max=0, min=-40}, outRange {max=0, min=1}}
    boneToModify {boneName=body_spine_01}
   }
   n28 = "Transform (Modify) Bone - Bone: body_spine_02"(
    ComponentPose=@n27
    Alpha=1
    Translation=Get spine_02_location() -> (spine_02_location:Vector)
    Rotation=Get spine_02_rotation() -> (spine_02_rotation:Rotator)
    Scale="1,1,1"
   ) {
    rotationMode=BMM_Replace
    rotationSpace=BCS_WorldSpace
    scaleSpace=BCS_BoneSpace
    translationMode=BMM_Replace
    translationSpace=BCS_WorldSpace
    alphaBoolBlend {blendOption=Linear}
    alphaScaleBiasClamp {inRange {max=0, min=-40}, outRange {max=0, min=1}}
    boneToModify {boneName=body_spine_02}
   }
   n29 = "Transform (Modify) Bone - Bone: body_spine_03"(
    ComponentPose=@n28
    Alpha=1
    Translation=Get spine_03_location() -> (spine_03_location:Vector)
    Rotation=Get spine_03_rotation() -> (spine_03_rotation:Rotator)
    Scale="1,1,1"
   ) {
    rotationMode=BMM_Replace
    rotationSpace=BCS_WorldSpace
    scaleSpace=BCS_BoneSpace
    translationMode=BMM_Replace
    translationSpace=BCS_WorldSpace
    alphaBoolBlend {blendOption=Linear}
    alphaScaleBiasClamp {inRange {max=0, min=-40}, outRange {max=0, min=1}}
    boneToModify {boneName=body_spine_03}
   }
   n30 = "Transform (Modify) Bone - Bone: body_spine_04"(
    ComponentPose=@n29
    Alpha=1
    Translation=Get spine_04_location() -> (spine_04_location:Vector)
    Rotation=Get spine_04_rotation() -> (spine_04_rotation:Rotator)
    Scale="1,1,1"
   ) {
    rotationMode=BMM_Replace
    rotationSpace=BCS_WorldSpace
    scaleSpace=BCS_BoneSpace
    translationMode=BMM_Replace
    translationSpace=BCS_WorldSpace
    alphaBoolBlend {blendOption=Linear}
    alphaScaleBiasClamp {inRange {max=0, min=-40}, outRange {max=0, min=1}}
    boneToModify {boneName=body_spine_04}
   }
   n31 = "Transform (Modify) Bone - Bone: body_spine_05"(
    ComponentPose=@n30
    Alpha=1
    Translation=Get spine_05_location() -> (spine_05_location:Vector)
    Rotation=Get spine_05_rotation() -> (spine_05_rotation:Rotator)
    Scale="1,1,1"
   ) {
    rotationMode=BMM_Replace
    rotationSpace=BCS_WorldSpace
    scaleSpace=BCS_BoneSpace
    translationMode=BMM_Replace
    translationSpace=BCS_WorldSpace
    alphaBoolBlend {blendOption=Linear}
    alphaScaleBiasClamp {inRange {max=0, min=-40}, outRange {max=0, min=1}}
    boneToModify {boneName=body_spine_05}
   }
   n32 = "Transform (Modify) Bone - Bone: body_clavicle_l"(
    ComponentPose=@n31
    Alpha=1
    Translation=Get clavicle_l_location() -> (clavicle_l_location:Vector)
    Rotation=Get clavicle_l_rotation() -> (clavicle_l_rotation:Rotator)
    Scale="1,1,1"
   ) {
    rotationMode=BMM_Replace
    rotationSpace=BCS_WorldSpace
    scaleSpace=BCS_BoneSpace
    translationMode=BMM_Replace
    translationSpace=BCS_WorldSpace
    alphaBoolBlend {blendOption=Linear}
    alphaScaleBiasClamp {inRange {max=0, min=-40}, outRange {max=0, min=1}}
    boneToModify {boneName=body_clavicle_l}
   }
   n33 = "Transform (Modify) Bone - Bone: body_clavicle_r"(
    ComponentPose=@n32
    Alpha=1
    Translation=Get clavicle_r_location() -> (clavicle_r_location:Vector)
    Rotation=Get clavicle_r_rotation() -> (clavicle_r_rotation:Rotator)
    Scale="1,1,1"
   ) {
    rotationMode=BMM_Replace
    rotationSpace=BCS_WorldSpace
    scaleSpace=BCS_BoneSpace
    translationMode=BMM_Replace
    translationSpace=BCS_WorldSpace
    alphaBoolBlend {blendOption=Linear}
    alphaScaleBiasClamp {inRange {max=0, min=-40}, outRange {max=0, min=1}}
    boneToModify {boneName=body_clavicle_r}
   }
   n34 = "Transform (Modify) Bone - Bone: body_upperarm_l"(
    ComponentPose=@n33
    Alpha=1
    Translation=Get upperarm_l_location() -> (upperarm_l_location:Vector)
    Rotation=Get upperarm_l_rotation() -> (upperarm_l_rotation:Rotator)
    Scale="1,1,1"
   ) {
    rotationMode=BMM_Replace
    rotationSpace=BCS_WorldSpace
    scaleSpace=BCS_BoneSpace
    translationMode=BMM_Replace
    translationSpace=BCS_WorldSpace
    alphaBoolBlend {blendOption=Linear}
    alphaScaleBiasClamp {inRange {max=0, min=-40}, outRange {max=0, min=1}}
    boneToModify {boneName=body_upperarm_l}
   }
   n35 = "Transform (Modify) Bone - Bone: body_upperarm_r"(
    ComponentPose=@n34
    Alpha=1
    Translation=Get upperarm_r_location() -> (upperarm_r_location:Vector)
    Rotation=Get upperarm_r_rotation() -> (upperarm_r_rotation:Rotator)
    Scale="1,1,1"
   ) {
    rotationMode=BMM_Replace
    rotationSpace=BCS_WorldSpace
    scaleSpace=BCS_BoneSpace
    translationMode=BMM_Replace
    translationSpace=BCS_WorldSpace
    alphaBoolBlend {blendOption=Linear}
    alphaScaleBiasClamp {inRange {max=0, min=-40}, outRange {max=0, min=1}}
    boneToModify {boneName=body_upperarm_r}
   }
   Save cached pose 'Initial_Cache'(Pose=Component To Local(ComponentPose=@n35)) {cacheName=Initial_Cache}
  ]}
 EventGraph {nodes [
   n1 = Event BlueprintUpdateAnimation()
   > n2 = Sequence!(execute) -> (then_0, then_1, then_2:Exec)
   > n3 = Gravity Override State Function DEPRECATED(CharacterPartAnimInstance=Self-Reference(), GameTimeInSeconds=GetGameTimeInSeconds(), GravityOverrideParams=Get Gravity Override Params())
   > Set GravityOverride(GravityOverride=@n3)
   n4 = Get bIsCrouching()
   n5 = Get bIsMoving2D()
   > Set BackpackAlpha(BackpackAlpha:double=SelectFloat(A=1, B=0.7, bPickA=AND Boolean(A=@n5, B=@n4))) -> (Output_Get:double)
   n6 = GetPartSkeletalMeshComponent()
   > Is Valid(InputObject=@n6)
   n7 = GetAttachParent!(self:Scene Component Object Reference=@n6) -> (ReturnValue:Scene Component Object Reference)
   > Is Valid(InputObject=@n7)
   > Set Body_Mesh(Body_Mesh:Scene Component Object Reference=@n7) -> (Output_Get:Scene Component Object Reference)
   n8 = Sequence!(execute=@n2.then_1) -> (then_0, then_1, then_2, then_3:Exec, then_4:Exec, then_5:Exec, then_6:Exec)
   n9 = Get Body_Mesh() -> (Body_Mesh:Scene Component Object Reference)
   n10 = BreakTransform(InTransform=GetSocketTransform(self=@n9, InSocketName=thigh_r, TransformSpace=RTS_World))
   Set thigh_right_location(execute=@n8.then_1, thigh_right_location:Vector=@n10) -> (Output_Get:Vector)
   > Set thigh_right_rotation(thigh_right_rotation:Rotator=@n10.Rotation) -> (Output_Get:Rotator)
   n11 = BreakTransform(InTransform=GetSocketTransform(self=@n9, InSocketName=calf_l, TransformSpace=RTS_World))
   > Set calf_left_location(calf_left_location:Vector=@n11) -> (Output_Get:Vector)
   > Set calf_left_rotation(calf_left_rotation:Rotator=@n11.Rotation) -> (Output_Get:Rotator)
   n12 = BreakTransform(InTransform=GetSocketTransform(self=@n9, InSocketName=calf_r, TransformSpace=RTS_World))
   Set calf_right_location(execute=@n8.then_2, calf_right_location:Vector=@n12) -> (Output_Get:Vector)
   > Set calf_right_rotation(calf_right_rotation:Rotator=@n12.Rotation) -> (Output_Get:Rotator)
   n13 = BreakTransform(InTransform=GetSocketTransform(self=@n9, InSocketName=root, TransformSpace=RTS_World))
   > Set body_root_location(body_root_location:Vector=@n13) -> (Output_Get:Vector)
   > Set body_root_rotation(body_root_rotation:Rotator=@n13.Rotation) -> (Output_Get:Rotator)
   n14 = BreakTransform(InTransform=GetSocketTransform(self=@n9, InSocketName=spine_01, TransformSpace=RTS_World))
   Set spine_01_location(execute=@n8.then_3, spine_01_location:Vector=@n14) -> (Output_Get:Vector)
   > Set spine_01_rotation(spine_01_rotation:Rotator=@n14.Rotation) -> (Output_Get:Rotator)
   n15 = BreakTransform(InTransform=GetSocketTransform(self=@n9, InSocketName=spine_02, TransformSpace=RTS_World))
   > Set spine_02_location(spine_02_location:Vector=@n15) -> (Output_Get:Vector)
   > Set spine_02_rotation(spine_02_rotation:Rotator=@n15.Rotation) -> (Output_Get:Rotator)
   n16 = BreakTransform(InTransform=GetSocketTransform(self=@n9, InSocketName=spine_03, TransformSpace=RTS_World))
   Set spine_03_location(execute=@n8.then_4, spine_03_location:Vector=@n16) -> (Output_Get:Vector)
   > Set spine_03_rotation(spine_03_rotation:Rotator=@n16.Rotation) -> (Output_Get:Rotator)
   n17 = BreakTransform(InTransform=GetSocketTransform(self=@n9, InSocketName=spine_04, TransformSpace=RTS_World))
   > Set spine_04_location(spine_04_location:Vector=@n17) -> (Output_Get:Vector)
   > Set spine_04_rotation(spine_04_rotation:Rotator=@n17.Rotation) -> (Output_Get:Rotator)
   n18 = BreakTransform(InTransform=GetSocketTransform(self=@n9, InSocketName=spine_05, TransformSpace=RTS_World))
   Set spine_05_location(execute=@n8.then_5, spine_05_location:Vector=@n18) -> (Output_Get:Vector)
   > Set spine_05_rotation(spine_05_rotation:Rotator=@n18.Rotation) -> (Output_Get:Rotator)
   n19 = BreakTransform(InTransform=GetSocketTransform(self=@n9, InSocketName=clavicle_l, TransformSpace=RTS_World))
   > Set clavicle_l_location(clavicle_l_location:Vector=@n19) -> (Output_Get:Vector)
   > Set clavicle_l_rotation(clavicle_l_rotation:Rotator=@n19.Rotation) -> (Output_Get:Rotator)
   n20 = BreakTransform(InTransform=GetSocketTransform(self=@n9, InSocketName=clavicle_r, TransformSpace=RTS_World))
   > Set clavicle_r_location(clavicle_r_location:Vector=@n20) -> (Output_Get:Vector)
   > Set clavicle_r_rotation(clavicle_r_rotation:Rotator=@n20.Rotation) -> (Output_Get:Rotator)
   n21 = BreakTransform(InTransform=GetSocketTransform(self=@n9, InSocketName=upperarm_l, TransformSpace=RTS_World))
   Set upperarm_l_location(execute=@n8.then_6, upperarm_l_location:Vector=@n21) -> (Output_Get:Vector)
   > Set upperarm_l_rotation(upperarm_l_rotation:Rotator=@n21.Rotation) -> (Output_Get:Rotator)
   n22 = BreakTransform(InTransform=GetSocketTransform(self=@n9, InSocketName=upperarm_r, TransformSpace=RTS_World))
   > Set upperarm_r_location(upperarm_r_location:Vector=@n22) -> (Output_Get:Vector)
   > Set upperarm_r_rotation(upperarm_r_rotation:Rotator=@n22.Rotation) -> (Output_Get:Rotator)
   n23 = BreakTransform(InTransform=GetSocketTransform(self=@n9, InSocketName=pelvis, TransformSpace=RTS_World))
   Set pelvis_location(execute=@n8.then_0, pelvis_location:Vector=@n23) -> (Output_Get:Vector)
   > Set pelvis_rotation(pelvis_rotation:Rotator=@n23.Rotation) -> (Output_Get:Rotator)
   n24 = BreakTransform(InTransform=GetSocketTransform(self=@n9, InSocketName=thigh_l, TransformSpace=RTS_World))
   > Set thigh_left_location(thigh_left_location:Vector=@n24) -> (Output_Get:Vector)
   > Set thigh_left_rotation(thigh_left_rotation:Rotator=@n24.Rotation) -> (Output_Get:Rotator)
   Set WPO Globals Function(
    execute=@n2.then_2
    Material=Get Mat Instance Dynamic()
    Overlay Material=Get Overlay Mat Instance Dynamic()
    Skeletal Mesh=Get SkelMesh()
    "Local Noise Transform  Location"=Get Local Noise Transform Location() -> (Local Noise Transform Location:Vector)
    Local Noise Offset=Get Local Noise Offset()
   )
   > Set WPO Regional Function!(
    execute
    Material:Material Instance Dynamic Object Reference=Get Mat Instance Dynamic()
    Overlay Material:Material Instance Dynamic Object Reference=Get Overlay Mat Instance Dynamic()
    Skeletal Mesh:Skeletal Mesh Component Object Reference=Get SkelMesh()
    Velocity Transform:Name=root
    Velocity Interp Speed:double=Get Velocity Interp Speed()
    Collision Transform:Name=pack
    Collision Transform Axis:EAxis Enum
    Invert Collision Transform Axis:bool=false
    Delta Time X:double=@n1.DeltaTimeX
    Previous Velocity Transform Location:"Vector (by ref)"=Get R1 Velocity Transform Location()
    Previous Velocity Transform Velocity:"Vector (by ref)"=Get R1 Velocity Transform Velocity()
    Region:int=1
   ) -> (then)
   > Set WPO Regional Function(
    Material=Get Mat Instance Dynamic()
    Overlay Material=Get Overlay Mat Instance Dynamic()
    Skeletal Mesh=Get SkelMesh()
    Velocity Transform=root
    Velocity Interp Speed=Get Velocity Interp Speed()
    Collision Transform=pack
    Invert Collision Transform Axis=true
    Delta Time X=@n1.DeltaTimeX
    Previous Velocity Transform Location=Get R2 Velocity Transform Location()
    Previous Velocity Transform Velocity=Get R2 Velocity Transform Velocity()
    Region=2
   )
   > n25 = Branch(Condition=Get bIsSkyDiving())
   > Set WPO Pho Wind Intensity(WPO Pho Wind Intensity=0.5)
   > n26 = Set WPO Pho Turbulence(WPO Pho Turbulence:double=0.9) -> (Output_Get:double)
   SetScalarParameterValue(execute=[@n28.then, @n26.then], self=Get Mat Instance Dynamic(), ParameterName=ClothSim_WindStrength, Value=Get WPO Pho Wind Intensity())
   n27 = Get WPO Pho Turbulence() -> (WPO Pho Turbulence:double)
   > SetScalarParameterValue(self=Get Mat Instance Dynamic(), ParameterName=ClothSim_R1_TurbulenceStrength, Value=@n27)
   > SetScalarParameterValue(self=Get Mat Instance Dynamic(), ParameterName=ClothSim_R2_TurbulenceStrength, Value=@n27)
   Set WPO Pho Wind Intensity(execute=@n25.else, WPO Pho Wind Intensity=0)
   > n28 = Set WPO Pho Turbulence(WPO Pho Turbulence=0.1)
   Event OnCharacterCustomizationLoaded()
   > n29 = Set SkelMesh(SkelMesh=GetPartSkeletalMeshComponent())
   > n30 = Get Dynamic Material Instance(SkeletalMeshComponent=@n29)
   > Set Mat Instance Dynamic(Mat Instance Dynamic=@n30)
   > n31 = GetOrCreateDynamicOverlayMaterial(MeshComponent=@n29, MaterialName=None)
   > Set Overlay Mat Instance Dynamic(Overlay Mat Instance Dynamic=@n31)
   TryGetPawnOwner()
  ]}
]
variables [
 BackpackAlpha {type=number}
 Body_Mesh {type=SceneComponent, default=null}
 body_root_location {type=Vector}
 body_root_rotation {type=Rotator}
 calf_left_location {type=Vector}
 calf_left_rotation {type=Rotator}
 calf_right_location {type=Vector}
 calf_right_rotation {type=Rotator}
 clavicle_l_location {type=Vector}
 clavicle_l_rotation {type=Rotator}
 clavicle_r_location {type=Vector}
 clavicle_r_rotation {type=Rotator}
 Gravity Override Params {
  type=GravityOverrideParamsStruct
  default {
   onGround_Moving {windAmplitude=<-75 -75 -75>, windFrequency=<4 4 4>, windOffset=<0 -200 0>}
   onGround_Standing {windAmplitude=<-75 -75 -75>, windFrequency=<2 2 2>, windOffset=<0 -100 0>}
   skydive_FloatingDown {windAmplitude=<150 -250 300>, windFrequency=<1.5 7 100>, windOffset=<25 25 25>}
  }
 }
 GravityOverride {type=Vector}
 Local Noise Offset {type=Vector}
 Local Noise Transform Location {type=Vector}
 Mat Instance Dynamic {type=MaterialInstanceDynamic}
 Overlay Mat Instance Dynamic {type=MaterialInstanceDynamic}
 pelvis_location {type=Vector}
 pelvis_rotation {type=Rotator}
 R1 Velocity Transform Location {type=Vector}
 R1 Velocity Transform Velocity {type=Vector}
 R2 Velocity Transform Location {type=Vector}
 R2 Velocity Transform Velocity {type=Vector}
 SkelMesh {type=SkeletalMeshComponent}
 spine_01_location {type=Vector}
 spine_01_rotation {type=Rotator}
 spine_02_location {type=Vector}
 spine_02_rotation {type=Rotator}
 spine_03_location {type=Vector}
 spine_03_rotation {type=Rotator}
 spine_04_location {type=Vector}
 spine_04_rotation {type=Rotator}
 spine_05_location {type=Vector}
 spine_05_rotation {type=Rotator}
 thigh_left_location {type=Vector}
 thigh_left_rotation {type=Rotator}
 thigh_right_location {type=Vector}
 thigh_right_rotation {type=Rotator}
 upperarm_l_location {type=Vector, default=<0 0 100>}
 upperarm_l_rotation {type=Rotator}
 upperarm_r_location {type=Vector}
 upperarm_r_rotation {type=Rotator}
 Velocity Interp Speed {type=number}
 WPO Pho Turbulence {type=number}
 WPO Pho Wind Intensity {type=number}
]
